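# Sound cue for dropping gold ingredient loot, layering a metal impact with a gold stack breaking sound.
# A Mixer plays one of five random metal drop impacts at reduced volume alongside one of four gold stack break variations delayed by a short fixed time.
# Each layer has its own upward random pitch modulation.
# It uses weapon drop attenuation and the loot collect sound class, with overridden concurrency that stops the oldest instance.
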SoundCue /Game/Sounds/Foley_Loot/Ingredients/Gold/Gold_Drop_Cue
root=Mixer_0
nodes [
 {id=Delay_1, children [Modulator_1], properties {delayMax=0.2, delayMin=0.2}}
 {id=Mixer_0, children [Modulator_0,Delay_1], properties {inputVolume [0.7,1]}}
 {id=Modulator_0, children [Random_0], properties {pitchMax=1.2, pitchMin=1}}
 {id=Modulator_1, children [Random_1], properties {pitchMax=1.2, pitchMin=1}}
 {
  id=Random_0
  children [WavePlayer_0,WavePlayer_1,WavePlayer_2,WavePlayer_3,WavePlayer_4]
  properties {bShouldExcludeFromBranchCulling=true, weights [1,1,1,1,1]}
 }
 {
  id=Random_1
  children [WavePlayer_5,WavePlayer_6,WavePlayer_7,WavePlayer_8]
  properties {bShouldExcludeFromBranchCulling=true, weights [1,1,1,1]}
 }
 {id=WavePlayer_0, properties {soundWaveAssetPtr=/Game/Sounds/Foley_Loot/Ingredients/Metal/Metal_Drop_Impact_01}}
 {id=WavePlayer_1, properties {soundWaveAssetPtr=/Game/Sounds/Foley_Loot/Ingredients/Metal/Metal_Drop_Impact_02}}
 {id=WavePlayer_2, properties {soundWaveAssetPtr=/Game/Sounds/Foley_Loot/Ingredients/Metal/Metal_Drop_Impact_03}}
 {id=WavePlayer_3, properties {soundWaveAssetPtr=/Game/Sounds/Foley_Loot/Ingredients/Metal/Metal_Drop_Impact_04}}
 {id=WavePlayer_4, properties {soundWaveAssetPtr=/Game/Sounds/Foley_Loot/Ingredients/Metal/Metal_Drop_Impact_05}}
 {id=WavePlayer_5, properties {soundWaveAssetPtr=&/GoldStack_Breaks_01}}
 {id=WavePlayer_6, properties {soundWaveAssetPtr=&/GoldStack_Breaks_02}}
 {id=WavePlayer_7, properties {soundWaveAssetPtr=&/GoldStack_Breaks_03}}
 {id=WavePlayer_8, properties {soundWaveAssetPtr=&/GoldStack_Breaks_04}}
]
properties {
 attenuationSettings=&Fort_Weapon_Drop_Att
 bOverrideConcurrency=true
 maxDistance=3000
 soundClassObject=&Fort_Loot_Collect
 volumeMultiplier=0.8
 concurrencyOverrides {maxCount=2, resolutionRule=StopOldest}
}
