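# Impact foley cue for gold loot ingredients, using weapon-drop attenuation and the loot-collect sound class.
# Mixes three randomized layers: dominant gold-stack impacts with varied pitch, generic metal impacts under a volume envelope rising from quiet, and faint, pitched-up big metal hits.
# Quality-level nodes switch the gold and generic metal layers between full and reduced variation pools.
# Concurrency override allows two instances, stopping the oldest.
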SoundCue /Game/Sounds/Foley_Loot/Ingredients/Gold/Gold_Impact_Cue
root=Mixer_0
nodes [
 {
  id=Enveloper_2
  children [QualityLevel_2]
  properties {
   pitchCurve {editorCurveData {keys [{time=0}]}}
   volumeCurve {editorCurveData {keys [
      {arriveTangent=-0.6248, interpMode=RCIM_Cubic, leaveTangent=-0.6248, tangentMode=RCTM_User, time=0, value=0.1}
      {arriveTangent=-0.0797, interpMode=RCIM_Cubic, leaveTangent=-0.0797, tangentMode=RCTM_User, time=1.2, value=1}
     ]}}
  }
 }
 {id=Mixer_0, children [Modulator_0,Enveloper_2,Modulator_1], properties {inputVolume [1.5,0.5,0.3]}}
 {id=Modulator_0, children [QualityLevel_0], properties {pitchMax=1.2, pitchMin=1}}
 {id=Modulator_1, children [Random_3], properties {pitchMax=1.2, pitchMin=1.2}}
 {id=QualityLevel_0, children [Random_0,Random_0,Random_1,Random_1]}
 {id=QualityLevel_2, children [Random_4,Random_4,Random_6,Random_6]}
 {
  id=Random_0
  children [
   WavePlayer_5
   WavePlayer_6
   WavePlayer_7
   WavePlayer_8
   WavePlayer_9
   WavePlayer_10
   WavePlayer_11
   WavePlayer_12
   WavePlayer_13
   WavePlayer_14
  ]
  properties {bShouldExcludeFromBranchCulling=true, weights [1,1,1,1,1,1,1,1,1,1]}
 }
 {
  id=Random_1
  children [WavePlayer_11, WavePlayer_12, WavePlayer_13, WavePlayer_14]
  properties {bShouldExcludeFromBranchCulling=true, weights [1,1,1,1]}
 }
 {id=Random_3, children [WavePlayer_26, WavePlayer_27, WavePlayer_28], properties {bShouldExcludeFromBranchCulling=true, weights [1,1,1]}}
 {
  id=Random_4
  children [WavePlayer_29, WavePlayer_30, WavePlayer_31, WavePlayer_32, WavePlayer_33, WavePlayer_34, WavePlayer_35]
  properties {bShouldExcludeFromBranchCulling=true, weights [1,1,1,1,1,1,1]}
 }
 {id=Random_6, children [WavePlayer_33, WavePlayer_34, WavePlayer_35], properties {bShouldExcludeFromBranchCulling=true, weights [1,1,1]}}
 {id=WavePlayer_10, properties {soundWaveAssetPtr=&/GoldStack_Impacts_06}}
 {id=WavePlayer_11, properties {soundWaveAssetPtr=&/GoldStack_Impacts_07}}
 {id=WavePlayer_12, properties {soundWaveAssetPtr=&/GoldStack_Impacts_08}}
 {id=WavePlayer_13, properties {soundWaveAssetPtr=&/GoldStack_Impacts_09}}
 {id=WavePlayer_14, properties {soundWaveAssetPtr=&/GoldStack_Impacts_10}}
 {id=WavePlayer_26, properties {soundWaveAssetPtr=&/Metal_Big_04}}
 {id=WavePlayer_27, properties {soundWaveAssetPtr=&/Metal_Big_03}}
 {id=WavePlayer_28, properties {soundWaveAssetPtr=&/Metal_Big_01}}
 {id=WavePlayer_29, properties {soundWaveAssetPtr=&/default_metal_impact_new_combined_01}}
 {id=WavePlayer_30, properties {soundWaveAssetPtr=&/default_metal_impact_new_combined_02}}
 {id=WavePlayer_31, properties {soundWaveAssetPtr=&/default_metal_impact_new_combined_03}}
 {id=WavePlayer_32, properties {soundWaveAssetPtr=&/default_metal_impact_new_combined_04}}
 {id=WavePlayer_33, properties {soundWaveAssetPtr=&/default_metal_impact_new_combined_05}}
 {id=WavePlayer_34, properties {soundWaveAssetPtr=&/default_metal_impact_new_combined_06}}
 {id=WavePlayer_35, properties {soundWaveAssetPtr=&/default_metal_impact_new_combined_07}}
 {id=WavePlayer_5, properties {soundWaveAssetPtr=&/GoldStack_Impacts_01}}
 {id=WavePlayer_6, properties {soundWaveAssetPtr=&/GoldStack_Impacts_02}}
 {id=WavePlayer_7, properties {soundWaveAssetPtr=&/GoldStack_Impacts_03}}
 {id=WavePlayer_8, properties {soundWaveAssetPtr=&/GoldStack_Impacts_04}}
 {id=WavePlayer_9, properties {soundWaveAssetPtr=&/GoldStack_Impacts_05}}
]
properties {
 attenuationSettings=&Fort_Weapon_Drop_Att
 bOverrideConcurrency=true
 maxDistance=3000
 soundClassObject=&Fort_Loot_Collect
 volumeMultiplier=0.5
 concurrencyOverrides {maxCount=2, resolutionRule=StopOldest}
}
